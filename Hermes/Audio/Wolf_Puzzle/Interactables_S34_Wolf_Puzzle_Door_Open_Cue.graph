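# Sound cue for the door opening in the S34 Wolf Puzzle interactables, routed to the Fort_Quests sound class.
# A single wave player feeds a one-input Mixer at the root, with a boosted cue volume.
# It reuses a water fountain focus-sphere attenuation asset for spatialization.
# A concurrency override allows one instance per owner and blocks new plays while one is active.
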SoundCue /Hermes/Audio/Wolf_Puzzle/Interactables_S34_Wolf_Puzzle_Door_Open_Cue
root=Mixer_0
nodes [
 {id=Mixer_0, children [WavePlayer_2], properties {inputVolume [1]}}
 {id=WavePlayer_2, properties {soundWaveAssetPtr=&/Interactables_S34_Wolf_Puzzle_DoorOpen}}
]
properties {
 attenuationSettings=/Game/Sounds/Helios/Environment_Actors/WaterFountain/Fort_Water_Fountain_Focus_Sphere_Att
 bOverrideConcurrency=true
 maxDistance=1250
 soundClassObject=&Fort_Quests
 volumeMultiplier=3
 concurrencyOverrides {bLimitToOwner=true, maxCount=1, resolutionRule=PreventNew}
}
