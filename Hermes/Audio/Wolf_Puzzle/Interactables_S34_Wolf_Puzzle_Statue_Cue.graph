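# Sound cue for the statue interactable in the S34 Wolf Puzzle, routed to the Fort_Quests sound class.
# It picks one of four equally weighted statue waves, applies slight pitch modulation, and passes it through a single-input mixer with boosted cue volume.
# A concurrency override allows one instance per owner and blocks new plays while it sounds.
# Attenuation reuses the HiddenStar ambient loop asset from SupplyDrops.
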SoundCue /Hermes/Audio/Wolf_Puzzle/Interactables_S34_Wolf_Puzzle_Statue_Cue
root=Mixer_0
nodes [
 {id=Mixer_0, children [Modulator_0], properties {inputVolume [1]}}
 {id=Modulator_0, children [Random_1], properties {pitchMax=1.02, pitchMin=0.99}}
 {id=Random_1, children [WavePlayer_2,WavePlayer_4,WavePlayer_5,WavePlayer_6], properties {weights [1,1,1,1]}}
 {id=WavePlayer_2, properties {soundWaveAssetPtr=&/Interactables_S34_Wolf_Puzzle_Statue_01}}
 {id=WavePlayer_4, properties {soundWaveAssetPtr=&/Interactables_S34_Wolf_Puzzle_Statue_02}}
 {id=WavePlayer_5, properties {soundWaveAssetPtr=&/Interactables_S34_Wolf_Puzzle_Statue_03}}
 {id=WavePlayer_6, properties {soundWaveAssetPtr=&/Interactables_S34_Wolf_Puzzle_Statue_04}}
]
properties {
 attenuationSettings=/Game/Athena/Sounds/SupplyDrops/Fort_HiddenStar_Ambient_Loop_Att
 bOverrideConcurrency=true
 maxDistance=2500
 soundClassObject=&Fort_Quests
 volumeMultiplier=2
 concurrencyOverrides {bLimitToOwner=true, maxCount=1, resolutionRule=PreventNew}
}
