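# Sound cue for the Wolf Puzzle stinger, played in the Fort_Quests sound class with boosted volume.
# A single stinger wave feeds through a one-entry Random node into a one-input Mixer, so there is currently no variation.
# It borrows the Hidden Star ambient loop attenuation from supply drops.
# A concurrency override allows one instance per owner, stopping the oldest when retriggered.
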SoundCue /Hermes/Audio/Wolf_Puzzle/Interactables_S34_Wolf_Puzzle_Stinger_Cue
root=Mixer_0
nodes [
 {id=Mixer_0, children [Random_1], properties {inputVolume [1]}}
 {id=Random_1, children [WavePlayer_0], properties {weights [1]}}
 {id=WavePlayer_0, properties {soundWaveAssetPtr=&/Interactables_S34_Wolf_Puzzle_Stinger_01}}
]
properties {
 attenuationSettings=/Game/Athena/Sounds/SupplyDrops/Fort_HiddenStar_Ambient_Loop_Att
 bOverrideConcurrency=true
 maxDistance=2500
 soundClassObject=&Fort_Quests
 volumeMultiplier=2
 concurrencyOverrides {bLimitToOwner=true, maxCount=1, resolutionRule=StopOldest}
}
